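# Idle cosmetic effect for the RevoltCrush character, adding pale flipbook wisps, puffed breath and glowing embers to its skeletal mesh.
# Body and Embers spawn across six bone groups; Body fades with owner speed, while Embers drift on curl noise with a second glow renderer.
# Breath pulses from another bone in a looped window, stops while running and changes shape per User.Variant.
# User flags hide foot, hand and head effects; time-of-day adjusts color.
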NiagaraSystem /BRCosmetics/Characters/Player/Male/Medium/Bodies/M_MED_RevoltCrush/FX/NS_Character_RevoltCrush_Idle
dependencies {
 usedDataInterfaces [
  &NiagaraDataInterfaceCurve
  &NiagaraDataInterfaceColorCurve
  &NiagaraDataInterfaceVectorCurve
  &NiagaraDataInterfaceSkeletalMesh
 ]
 usedDynamicInputs [
  &Vector2DFromFloat
  &MakeVectorFromLinearColorRGB
  &ColorFromCurve
  &Vector4FromCurve
  &FNTimeOfDayScalar
  &SetBoolByFloatComparison
 ]
 usedModules [
  &self:SystemSpawnScript.SystemScriptSource.SystemScriptGraph.NiagaraNodeAssignment_0.SetVariables_#1
  &SystemState
  &self:SystemSpawnScript.SystemScriptSource.SystemScriptGraph.NiagaraNodeAssignment_3.SetVariables_#2
  &self:SystemSpawnScript.SystemScriptSource.SystemScriptGraph.NiagaraNodeAssignment_2.SetVariables_#3
 ]
}
emitters {
 Body {
  emitterUpdateScript [
   &EmitterState(
    Life Cycle Mode=::NewEnumerator1
    Loop Duration=&RandomRangeFloat(Minimum=3, Maximum=5)
    UseLoopDelay=false
   )
   &SpawnRate(
    SpawnRate=&Multiply_Float(A=&MakeCustomFloatFromBool(Bool=$System.IsRunning, True Float=1.5, False Float=1.5), B=6)
    Use Spawn Probability=false
   ) {module=SpawnRate001}
  ]
  particleSpawnScript [
   &FN_SetSpawnGroup(Number of Spawn Groups=6)
   &InitializeParticle(
    Color=<rgba 1.3344 1.3758 1.4 0.134>
    Lifetime=&RandomRangeFloat(Minimum=2, Maximum=2.5)
    Sprite Rotation Mode=::NewEnumerator2
    Sprite Size Mode=::NewEnumerator3
    Sprite UV Mode=::NewEnumerator3
    Uniform Sprite Size=&MergeHelper_FloatSwitch(0=18, 1=18, 2=13, 3=13, 4=14, 5=14, Index=&SpawnGroup(), Num=::NewEnumerator4)
    UsePositionOffset=false
    Write Material Random=true
   )
   &CameraOffset(Camera Offset Amount=8)
  ]
  particleUpdateScript [
   &ParticleState()
   &SkeletalMeshLocation(
    Bone / Socket Index=&SpawnGroup()
    Bone Sampling Mode=::NewEnumerator4
    MaskSpawnGroup=false
    Sampled Position Offset=&Add_Vector(
     A=&MergeHelper_VectorSwitch(
      0=&RandomRangeVector(Minimum=<-7 8 -1>, Maximum=<-7 8 1>)
      1=&RandomRangeVector(Minimum=<7 -8 -1>, Maximum=<7 -8 1>)
      2=<0 0 0>
      3=<0 0 0>
      4=<-3 0 0>
      5=<3 0 0>
      Index=&SpawnGroup()
      Num=::NewEnumerator4
     )
     B=&MergeHelper_VectorSwitch(
      0=<0 0 0>
      1=<0 0 0>
      2=&VectorFromCurve(
       VectorCurve=DataInterface() {propertyValues {XCurve {keys [<at 0 0>,<at 1 1>]}, YCurve {keys [<at 0 0>,<at 1 0>]}, ZCurve {keys [<at 0 0>,<at 1 0>]}}}
       Scale Curve=<5 1 1>
      )
      3=&VectorFromCurve(
       VectorCurve=DataInterface() {propertyValues {XCurve {keys [<at 0 0>,<at 1 1>]}, YCurve {keys [<at 0 0>,<at 1 0>]}, ZCurve {keys [<at 0 0>,<at 1 0>]}}}
       Scale Curve=<-5 1 1>
      )
      4=&VectorFromCurve(
       VectorCurve=DataInterface() {propertyValues {XCurve {keys [<at 0 0>,<at 1 1>]}, YCurve {keys [<at 0 0>,<at 1 0>]}, ZCurve {keys [<at 0 0>,<at 1 0>]}}}
       Scale Curve=<12 1 1>
      )
      5=&VectorFromCurve(
       VectorCurve=DataInterface() {propertyValues {XCurve {keys [<at 0 0>,<at 1 1>]}, YCurve {keys [<at 0 0>,<at 1 0>]}, ZCurve {keys [<at 0 0>,<at 1 0>]}}}
       Scale Curve=<-12 1 1>
      )
      Index=&SpawnGroup()
      Num=::NewEnumerator4
     )
    )
    Skeletal Mesh=$System.Skeletal Mesh
   ) {module=SkeletalMeshLocation002}
   &ScaleSpriteSize(
    Uniform Curve Scale=&Multiply_Float(
     A=1.7675
     B=&MergeHelper_FloatSwitch(
      0=&MakeCustomFloatFromBool(Bool=$User.bIsFootHidden, True Float=0, False Float=1)
      1=&MakeCustomFloatFromBool(Bool=$User.bIsFootHidden, True Float=0, False Float=1)
      2=&MakeCustomFloatFromBool(Bool=$User.bIsHandHidden, True Float=0, False Float=1)
      3=&MakeCustomFloatFromBool(Bool=$User.bIsHandHidden, True Float=0, False Float=1)
      4=1
      5=1
      Index=&SpawnGroup()
      Num=::NewEnumerator4
     )
    )
    Uniform Curve Sprite Scale=DataInterface() {propertyValues {Curve {keys [<at 0 0.5477 RCIM_Cubic>, {arriveTangent=-0.0123, tangentMode=RCTM_User, time=1, value=1}]}}}
   )
   &ScaleColor(
    Scale Alpha=&FloatFromCurve(
     FloatCurve=DataInterface() {propertyValues {Curve {keys [
         {arriveTangent=0.9709, interpMode=RCIM_Cubic, leaveTangent=0.9709, tangentMode=RCTM_User, time=0, value=0}
         {arriveTangent=0.4522, interpMode=RCIM_Cubic, leaveTangent=0.4522, tangentMode=RCTM_User, time=0.3235, value=0.922}
         {arriveTangent=-1.4219, interpMode=RCIM_Cubic, leaveTangent=-1.4219, tangentMode=RCTM_User, time=0.6427, value=0.8168}
         {arriveTangent=0.0143, interpMode=RCIM_Cubic, leaveTangent=0.0143, tangentMode=RCTM_User, time=1, value=0}
        ]}}}
     Scale Curve=&Multiply_Float(
      A=&MergeHelper_FloatSwitch(0=0.7, 1=0.7, 2=0.56, 3=0.56, 4=0.56, 5=0.56, Index=&SpawnGroup(), Num=::NewEnumerator4)
      B=&Multiply_Float(A=&FN_FrontEndFloatModifier(Gameplay=1, Front End=0.9), B=1)
     )
    )
    ScaleA=true
    ScaleRGB=true
   )
   /SetVariables(Particles.SubImageIndex=&FloatFromCurve(FloatCurve=DataInterface() {propertyValues {Curve {keys [<at 0 34.433>,<at 1 53.661>]}}})) {isSetParameters=true, module=SetVariables_#4}
   &ScaleColor(
    Color Value To Scale=$Particles.Color
    Scale Mode=::NewEnumerator1
    Scale RGBA=&MakeVector4_FromLinearColor(LinearColor=&FNTimeOfDayColor(
      Night Color=<rgba 0.539 0.6941 1 1>
      Morning Color=<rgba 0.8743 0.9172 1 1>
      Day Color=<rgba 1 1 1 1>
      Evening Color=<rgba 0.8802 0.8802 0.8802 1>
     ))
    ScaleRGBA=true
   ) {module=ScaleColor001}
   &DynamicMaterialParameters(
    Index 0 Param 1=3
    Index 0 Param 2=1
    Index 0 Param 3=0.5
    Module.Index0Float0Write=true
    Module.Index0Float1Write=true
    Module.Index0Float2Write=true
    Module.Index0Float3Write=true
   )
   &SolveForcesAndVelocity(Clamp Velocity=false, Limit Acceleration=false)
   &ScaleColor(
    Color Value To Scale=$Particles.Color
    Scale Alpha=&Lerp_Float(
     A=0.6
     B=1
     Alpha=&RemapRange(
      Input Value To Remap=&VectorLength(Vector=$Engine.Owner.Velocity)
      Input Min=0
      Input Max=200
      Output Min=1
      Output Max=0
     )
    )
    ScaleA=true
    ScaleRGB=true
   ) {module=ScaleColor004}
  ]
  rendererClasses [&NiagaraSpriteRendererProperties]
  renderers [{rendererClass=&NiagaraSpriteRendererProperties}]
 }
 Breath {
  emitterUpdateScript [
   &EmitterState(
    Life Cycle Mode=::NewEnumerator1
    Loop Duration=&RandomRangeFloat(Minimum=3.5, Maximum=4)
    UseLoopDelay=false
   )
   &SpawnRate(
    SpawnRate=&MakeCustomFloatFromBool(
     Bool=$User.bIsHeadHidden
     True Float=0
     False Float=&Multiply_Float(
      A=&MakeCustomFloatFromBool(Bool=$System.IsRunning, True Float=0, False Float=20)
      B=&FloatFromCurve(
       FloatCurve=DataInterface() {propertyValues {Curve {keys [<at 0.3633 0 RCIM_Constant>,<at 0.4034 1 RCIM_Constant>,<at 0.5724 0 RCIM_Constant>]}}}
       CurveIndex=$Emitter.NormalizedLoopAge
      )
     )
    )
    Use Spawn Probability=false
   )
  ]
  particleSpawnScript [
   &FN_SetSpawnGroup(Number of Spawn Groups:NiagaraInt32=&MergeHelper_IntSwitch(
     0=1
     1=2
     2=1
     3=?
     4=?
     5=?
     Index=&MakeIntFromFloat(Float=$User.Variant, Conversion Mode=::NewEnumerator1)
     Num=::NewEnumerator1
    ))
   &InitializeParticle(
    Color=<rgba 0.9688 0.9885 1 0.13>
    Lifetime=&RandomRangeFloat(Minimum=1.3, Maximum=1.3)
    Sprite Rotation Mode=::NewEnumerator2
    Sprite Size Mode=::NewEnumerator3
    Sprite UV Mode=::NewEnumerator3
    Uniform Sprite Size=24
    UsePositionOffset=false
    Write Material Random=true
   )
   &CameraOffset(Camera Offset Amount=5)
   &SkeletalMeshLocation(
    Bone / Socket Index=6
    Bone Sampling Mode=::NewEnumerator4
    MaskSpawnGroup=false
    Sampled Position Offset=&MergeHelper_VectorSwitch(
     0=<-3 6 0>
     1=&MergeHelper_VectorSwitch(0=<2 10.5 1>, 1=<2 10.5 -1>, 2=?, 3=?, Index=&SpawnGroup(), Num=::NewEnumerator0)
     2=<-3 6 0>
     3=?
     Index=&MakeIntFromFloat(Float=$User.Variant, Conversion Mode=::NewEnumerator1)
     Num=::NewEnumerator1
    )
    Skeletal Mesh=$System.Skeletal Mesh
   )
   &AddVelocity(
    Rotation Coordinate Space=::World
    Velocity=&RotateVectorByQuaternion(
     Vector To Rotate=&MergeHelper_VectorSwitch(
      0=&RandomRangeVector(Minimum=<0 13 0>, Maximum=<0 16 0>)
      1=&RandomRangeVector(Minimum=<-9 2 0>, Maximum=<-7 3 0>)
      2=&RandomRangeVector(Minimum=<0 13 0>, Maximum=<0 16 0>)
      3=?
      Index=&MakeIntFromFloat(Float=$User.Variant, Conversion Mode=::NewEnumerator1)
      Num=::NewEnumerator1
     )
     Quaternion=$Particles.SkeletalMeshLocation.SampledRotation
    )
    Velocity Speed Scale=1.6
   )
  ]
  particleUpdateScript [
   &ParticleState()
   &ScaleSpriteSize(
    Uniform Curve Scale=&MergeHelper_FloatSwitch(
     0=1.3
     1=0.7
     2=1.3
     3=?
     Index=&MakeIntFromFloat(Float=$User.Variant, Conversion Mode=::NewEnumerator1)
     Num=::NewEnumerator1
    )
    Uniform Curve Sprite Scale=DataInterface() {propertyValues {Curve {keys [<at 0 0.0946 RCIM_Cubic>, {arriveTangent=1.5714, tangentMode=RCTM_User, time=1, value=1}]}}}
   )
   &ScaleColor(
    Scale Alpha=&FloatFromCurve(
     FloatCurve=DataInterface() {propertyValues {Curve {keys [
         {arriveTangent=0.0587, interpMode=RCIM_Cubic, leaveTangent=0.0587, tangentMode=RCTM_User, time=0, value=0}
         {arriveTangent=0.3824, interpMode=RCIM_Cubic, leaveTangent=0.3824, tangentMode=RCTM_User, time=0.2576, value=0.922}
         {arriveTangent=-1.4219, interpMode=RCIM_Cubic, leaveTangent=-1.4219, tangentMode=RCTM_User, time=0.6427, value=0.8168}
         {arriveTangent=0.0143, interpMode=RCIM_Cubic, leaveTangent=0.0143, tangentMode=RCTM_User, time=1, value=0}
        ]}}}
     Scale Curve=&Multiply_Float(
      A=0.55
      B=&FloatFromCurve(
       FloatCurve=DataInterface() {propertyValues {Curve {keys [
           <at 0.3633 1 RCIM_Cubic>
           {arriveTangent=-0.0527, interpMode=RCIM_Cubic, leaveTangent=-0.0527, tangentMode=RCTM_User, time=0.6101, value=1}
           <at 0.8273 0 RCIM_Cubic>
          ]}}}
       CurveIndex=$Emitter.NormalizedLoopAge
      )
     )
    )
    ScaleA=true
    ScaleRGB=true
   )
   /SetVariables(Particles.SubImageIndex=&FloatFromCurve(FloatCurve=DataInterface() {propertyValues {Curve {keys [<at 0 30>,<at 1 64>]}}})) {isSetParameters=true, module=SetVariables_#4}
   &ScaleColor(
    Color Value To Scale=$Particles.Color
    Scale Mode=::NewEnumerator1
    Scale RGBA=&MakeVector4_FromLinearColor(LinearColor=&FNTimeOfDayColor(
      Night Color=<rgba 0.539 0.6941 1 1>
      Morning Color=<rgba 0.8743 0.9172 1 1>
      Day Color=<rgba 1 1 1 1>
      Evening Color=<rgba 0.8854 0.8854 0.8854 1>
     ))
    ScaleRGBA=true
   ) {module=ScaleColor001}
   &DynamicMaterialParameters(
    Index 0 Param 1=3
    Index 0 Param 2=1
    Index 0 Param 3=0.5
    Module.Index0Float0Write=true
    Module.Index0Float1Write=true
    Module.Index0Float2Write=true
    Module.Index0Float3Write=true
   )
   &GravityForce(Gravity=&Multiply_VectorByFloat(
     Vector=&RandomRangeVector(Minimum=<0 0 37>, Maximum=<0 0 41>)
     Float=&FloatFromCurve(
      FloatCurve=DataInterface() {propertyValues {Curve {keys [{arriveTangent=1.3787, interpMode=RCIM_Cubic, leaveTangent=1.3787, tangentMode=RCTM_User, time=0, value=0.2}, <at 1 1>]}}}
      Scale Curve=&MergeHelper_FloatSwitch(
       0=1
       1=0
       2=1
       3=?
       Index=&MakeIntFromFloat(Float=$User.Variant, Conversion Mode=::NewEnumerator1)
       Num=::NewEnumerator1
      )
     )
    ))
   &InheritSourceMovement(
    Applied Position Scale=&VectorFromFloat(Value=&FloatFromCurve(FloatCurve=DataInterface() {propertyValues {Curve {keys [<at 0 1 RCIM_Cubic>,<at 1 0>]}}}))
    ApplyForce=false
    ApplyPosition=true
    ApplyPositionFalloff=false
   )
   &SolveForcesAndVelocity(Clamp Velocity=false, Limit Acceleration=false)
  ]
  rendererClasses [&NiagaraSpriteRendererProperties]
  renderers [{rendererClass=&NiagaraSpriteRendererProperties}]
 }
 Embers {
  emitterUpdateScript [
   &EmitterState()
   &SpawnRate(
    SpawnRate=&MakeCustomFloatFromBool(Bool=$System.IsRunning, True Float=24, False Float=24)
    Use Spawn Probability=false
   ) {module=SpawnRate001}
  ]
  particleSpawnScript [
   /SetVariables(Particles.PerParticleRandom:NiagaraFloat=&RandomRangeFloat(Minimum=0, Maximum=1)) {isSetParameters=true}
   &FN_SetSpawnGroup(Number of Spawn Groups=6)
   &InitializeParticle(
    Color=&MultiplyLinearColor(Linear Color=<rgba 1 1 1 0.23>, Scale Factor=&MakeVector4_FromFloat(Float=$System.TOD))
    Lifetime=&RandomRangeFloat(Minimum=1, Maximum=1.4)
    Sprite Size=&Multiply_Vector2D(A=<0.4 0.4>, B=<1 1>)
    Sprite Size Mode=::NewEnumerator4
    UsePositionOffset=false
    Write Material Random=true
   )
   &SkeletalMeshLocation(
    Bone / Socket Index=&SpawnGroup()
    Bone Sampling Mode=::NewEnumerator4
    MaskSpawnGroup=false
    Sampled Position Offset=&MergeHelper_VectorSwitch(
     0=&RandomRangeVector(Minimum=<-7 8 -1>, Maximum=<-7 8 1>)
     1=&RandomRangeVector(Minimum=<7 -8 -1>, Maximum=<7 -8 1>)
     2=&RandomRangeVector(Minimum=<-5 -5 -5>, Maximum=<5 5 5>)
     3=&RandomRangeVector(Minimum=<-5 -5 -5>, Maximum=<5 5 5>)
     4=&RandomRangeVector(Minimum=<-5 -5 -5>, Maximum=<5 5 5>)
     5=&RandomRangeVector(Minimum=<-5 -5 -5>, Maximum=<5 5 5>)
     Index=&SpawnGroup()
     Num=::NewEnumerator4
    )
    Skeletal Mesh=$System.Skeletal Mesh
   )
   &AddVelocity(
    Rotation Coordinate Space=::World
    Velocity=&RandomRangeVector(Minimum=<-20 -20 -20>, Maximum=<20 20 20>)
    Velocity Speed Scale=0.6
   )
  ]
  particleUpdateScript [
   &ParticleState()
   &Drag(
    Drag=&FloatFromCurve(
     FloatCurve=DataInterface() {propertyValues {Curve {keys [
         {arriveTangent=0.1577, interpMode=RCIM_Cubic, leaveTangent=0.1577, tangentMode=RCTM_User, time=0, value=0.125}
         {arriveTangent=1.6704, interpMode=RCIM_Cubic, leaveTangent=1.6704, tangentMode=RCTM_User, time=1, value=1}
        ]}}}
     Scale Curve=&RandomRangeFloat(Minimum=2.5, Maximum=4)
    )
    Use Linear Drag=true
    Use Rotational Drag=true
   )
   &CurlNoiseForce(
    Bias Noise=false
    Noise Frequency=55
    Noise Strength=&FloatFromCurve(
     FloatCurve=DataInterface() {propertyValues {Curve {keys [{interpMode=RCIM_Cubic, tangentMode=RCTM_User, time=0, value=0.2}, {interpMode=RCIM_Cubic, tangentMode=RCTM_User, time=1, value=1}]}}}
     Scale Curve=96.926
    )
    Pan Noise=false
    Random Seed=33
   ) {module=CurlNoiseForce001}
   &SolveForcesAndVelocity(Clamp Velocity=false, Limit Acceleration=false)
   &ScaleColor(
    Linear Color Curve=DataInterface() {propertyValues {
      AlphaCurve {keys [<at 0 0>,<at 0.4658 0.35 RCIM_Cubic>,<at 0.9968 0>]}
      BlueCurve {keys [<at 0.0028 5>, {arriveTangent=-4.2116, interpMode=RCIM_Cubic, leaveTangent=-4.2116, time=0.2108, value=3}, <at 1 0.8 RCIM_Cubic>]}
      GreenCurve {keys [<at 0.0028 4.401>, {arriveTangent=-3.6569, interpMode=RCIM_Cubic, leaveTangent=-3.6569, time=0.2108, value=2.7927}, <at 1 0.7542 RCIM_Cubic>]}
      RedCurve {keys [
        <at 0.0028 3.8021>
        {arriveTangent=-3.1023, interpMode=RCIM_Cubic, leaveTangent=-3.1023, time=0.2108, value=2.2969}
        <at 1 0.7083 RCIM_Cubic>
       ]}
     }}
    Scale Mode=::NewEnumerator2
   )
   &ScaleSpriteSize(
    Uniform Curve Scale=&Multiply_Float(
     A=&RandomRangeFloat(Minimum=2.3, Maximum=2.6)
     B=&MergeHelper_FloatSwitch(
      0=&MakeCustomFloatFromBool(Bool=$User.bIsFootHidden, True Float=0, False Float=1)
      1=&MakeCustomFloatFromBool(Bool=$User.bIsFootHidden, True Float=0, False Float=1)
      2=&MakeCustomFloatFromBool(Bool=$User.bIsHandHidden, True Float=0, False Float=1)
      3=&MakeCustomFloatFromBool(Bool=$User.bIsHandHidden, True Float=0, False Float=1)
      4=1
      5=1
      Index=&SpawnGroup()
      Num=::NewEnumerator4
     )
    )
    Uniform Curve Sprite Scale=DataInterface() {propertyValues {Curve {keys [
        {tangentMode=RCTM_User, time=0, value=1}
        {arriveTangent=-0.0405, interpMode=RCIM_Cubic, leaveTangent=-0.0405, tangentMode=RCTM_User, time=0.7009, value=1}
        {arriveTangent=-8.0089, interpMode=RCIM_Cubic, leaveTangent=-8.0089, tangentMode=RCTM_User, time=1, value=0}
       ]}}}
   )
   /SetVariables(Particles.SpriteSizeGlow:Vector2f=&Multiply_Vector2D_ByFloat(Float=1.2, Vector2D=$Particles.SpriteSize)) {isSetParameters=true}
   /SetVariables(Particles.Glow Color=&MultiplyLinearColor(Linear Color=$Particles.Color, Scale Factor=<2 2 2 0.15>)) {isSetParameters=true}
  ]
  rendererClasses [&NiagaraSpriteRendererProperties]
  renderers [
   {rendererClass=&NiagaraSpriteRendererProperties}
   {rendererClass=&NiagaraSpriteRendererProperties, rendererIndex=1}
  ]
 }
}
summary {
 systemName=NS_Character_RevoltCrush_Idle
 emitters [
  {emitterName=Breath, rendererClasses [&NiagaraSpriteRendererProperties]}
  {emitterName=Embers, rendererClasses [&NiagaraSpriteRendererProperties]}
  {emitterName=Body, rendererClasses [&NiagaraSpriteRendererProperties]}
 ]
 userVariables [
  User.bIsFootHidden {type=NiagaraBool}
  User.bIsFrontEnd {type=NiagaraBool}
  User.bIsFrontEndPreview {type=NiagaraBool}
  User.bIsHandHidden {type=NiagaraBool}
  User.bIsHeadHidden {type=NiagaraBool}
  User.Variant {type=NiagaraFloat, defaultValue=0, description=""}
 ]
}
