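# Physics asset for the RevoltCrush male medium player body cosmetic, covering only an axe accessory and the right leg.
# The Axe body uses two capsules and is constrained to a tiny Base_Axe anchor capsule with tight swing and twist limits.
# Large capsules on the right thigh and calf form leg collision volumes, joined by a constraint with free swing and twist.
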PhysicsAsset /BRCosmetics/Characters/Player/Male/Medium/Bodies/M_MED_RevoltCrush/Meshes/M_MED_RevoltCrush_Physics
bodies {
 Axe [
  {shapeType=Capsule, length=27.7768, radius=2.3533, shapeName=Axe_capsule, center=<18.7461 -2.0401 -3.134>, rotation=<rot 0 -98.9493 -86.9542>}
  {shapeType=Capsule, length=11, radius=6, shapeName=Axe_capsule, center=<0 5.4544 -4.9667>, rotation=<rot 0 0.0001 74.7839>}
 ]
 Base_Axe [{shapeType=Capsule, length=0.505, radius=0.505, shapeName=Base_Axe_capsule, rotation=<rot 0 -90 0>}]
 calf_r [{shapeType=Capsule, length=25.5, radius=15.5, shapeName=calf_r_capsule, center=<20.7479 2.2778 -9.524>, rotation=<rot 90 0 0>}]
 thigh_r [{shapeType=Capsule, length=28.5, radius=12.5, shapeName=thigh_r_capsule, center=<17.721 0 -7.3214>, rotation=<rot 90 0 0>}]
}
constraints [
 {bone1Name=Axe, bone2Name=Base_Axe, swing1LimitDegrees=1, swing2LimitDegrees=10, twistLimitDegrees=10}
 {
  bone1Name=calf_r
  bone2Name=thigh_r
  swing1LimitDegrees=45
  swing1Motion=Free
  swing2LimitDegrees=45
  swing2Motion=Free
  twistLimitDegrees=45
  twistMotion=Free
 }
]
